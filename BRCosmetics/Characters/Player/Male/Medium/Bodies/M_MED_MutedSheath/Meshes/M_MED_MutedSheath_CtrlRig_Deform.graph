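# Deformation Control Rig for the MutedSheath male medium player body cosmetic, correcting helper and accessory bones as the body moves.
# It runs a shared DeformRig function, then parent-constrains pelvis, chest, card, cup and arm-bag bones to weighted body bones.
# Spine bend drives chest bag offsets.
# Per-thigh spherical pose readers, mostly reshaped by curves, blend jacket, card and katana offsets, with glute deform translation adding back-jacket correction.
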ControlRigBlueprint /BRCosmetics/Characters/Player/Male/Medium/Bodies/M_MED_MutedSheath/Meshes/M_MED_MutedSheath_CtrlRig_Deform
graphs [
 RigVMFunctionLibrary {links [], nodes []}
 RigVMFunction_Sequence_1_2_ContainedGraph {nodes [
   n1 = RigVMFunctionEntryNode()
   n2 = &RigVMFunction_Sequence(ExecuteContext=@n1.ExecuteContext)
   n3 = &RigVMFunction_Sequence(ExecuteContext=@n2.B)
   n4 = &RigVMFunction_Sequence(ExecuteContext=@n3.B)
   RigVMFunctionReturnNode(A=@n2.A, B=@n3.A, C=@n4.A, D=@n4.B)
  ]}
 RigVMModel {nodes [
   RigUnit_BeginExecution = &RigUnit_BeginExecution()
   RigVMFunction_Sequence_1 = RigVMAggregateNode(ExecuteContext=@RigUnit_BeginExecution.ExecutePin)
   FN_DeformRig_FNC = RigVMFunctionReferenceNode(
    DeformRig_ Curve="(deformRig_l_upperarm_up_alpha_72_#1=False,deformRig_r_upperarm_up_alpha_73_#2=False,deformRig_l_upperarm_fr_alpha_74_#3=False,deformRig_r_upperarm_fr_alpha_75_CB51B8 ...(truncated)"
    DeformRig_Bool="(upperarm_1_#4=True,elbow_3_#5=True,wrist_7_#6=True,groin_8_#7=True,glute_17_#8=True,knee_9_40C408324CA261 ...(truncated)"
    DeformRig_Debug=(upperarm_debug_18_#4=False,wrist_debug_20_#6=False,lat_debug_26_#9=False,groinGlute_debug_27_#7=False)
    DeformRig_Value="(upperarm_up_18_#10=(Rotation=(X=0,Y=0,Z=0,W=1),Translation=(X=0,Y=0,Z=8),Scale3D=(X=0.7,Y=1,Z=1)),upperarm_fwd_fr_19_#11=(Rotation=(X=0,Y=0,Z=0,W=1),Translation=(X=0,Y=10,Z=0),Scale3D=(X=0.5,Y=1,Z= ...(truncated)"
    ExecuteContext=@RigVMFunction_Sequence_1.A
   )
   &RigUnit_ParentConstraint(
    AdvancedSettings=(InterpolationType=Average,RotationOrderForFilter=XZY)
    Child=(Type=Bone,Name="extra_pelvis_bag_back")
    Parents=((Item=(Type=Bone,Name="pelvis"),Weight=0.879),(Item=(Type=Bone,Name="spine_01"),Weight=0.044),(Item=(Type=Bone,Name="spine_02"),Weight=0.044),(Item=(Type=Bone,Name="spine_03"),Weight=0.03))
    Weight=1
    ExecutePin=@RigVMFunction_Sequence_1.B
   )
   > &RigUnit_ParentConstraint(
    AdvancedSettings=(InterpolationType=Average,RotationOrderForFilter=XZY)
    Child=(Type=Bone,Name="dyn_card_BASE")
    Parents=((Item=(Type=Bone,Name="pelvis"),Weight=0.914),(Item=(Type=Bone,Name="spine_01"),Weight=0.084))
    Weight=1
   )
   > &RigUnit_ParentConstraint(
    AdvancedSettings=(InterpolationType=Average,RotationOrderForFilter=XZY)
    Child=(Type=Bone,Name="extra_chest_bag_bot")
    Parents=((Item=(Type=Bone,Name="pelvis"),Weight=0.052),(Item=(Type=Bone,Name="spine_01"),Weight=0.104),(Item=(Type=Bone,Name="spine_02"),Weight=0.278),(Item=(Type=Bone,Name="spine_03"),Weight=0.429),(Item=(Type=Bone,Name="spine_04"),Weight=0.138))
    Weight=1
   )
   > &RigUnit_ParentConstraint(
    AdvancedSettings=(InterpolationType=Average,RotationOrderForFilter=XZY)
    Child=(Type=Bone,Name="extra_chest_bag_top")
    Filter=(TranslationFilter=(bX=True,bY=True,bZ=True),RotationFilter=(bX=false,bY=false,bZ=false),ScaleFilter=(bX=True,bY=True,bZ=True))
    Parents=((Item=(Type=Bone,Name="spine_01"),Weight=0.027),(Item=(Type=Bone,Name="spine_02"),Weight=0.147),(Item=(Type=Bone,Name="spine_03"),Weight=0.306),(Item=(Type=Bone,Name="spine_04"),Weight=0.4),(Item=(Type=Bone,Name="spine_05"),Weight=0.121))
    Weight=1
   )
   > &RigUnit_ParentConstraint(
    AdvancedSettings=(InterpolationType=Average,RotationOrderForFilter=XZY)
    Child=(Type=Bone,Name="extra_cup1")
    Filter=(TranslationFilter=(bX=True,bY=True,bZ=True),RotationFilter=(bX=false,bY=false,bZ=false),ScaleFilter=(bX=True,bY=True,bZ=True))
    Parents=((Item=(Type=Bone,Name="spine_01"),Weight=0.032),(Item=(Type=Bone,Name="spine_02"),Weight=0.152),(Item=(Type=Bone,Name="spine_03"),Weight=0.32),(Item=(Type=Bone,Name="spine_04"),Weight=0.382),(Item=(Type=Bone,Name="spine_05"),Weight=0.114))
    Weight=1
   )
   > &RigUnit_ParentConstraint(
    AdvancedSettings=(InterpolationType=Average,RotationOrderForFilter=XZY)
    Child=(Type=Bone,Name="extra_arm_bag_front")
    Filter=(TranslationFilter=(bX=True,bY=True,bZ=True),RotationFilter=(bX=false,bY=false,bZ=false),ScaleFilter=(bX=True,bY=True,bZ=True))
    Parents="((Item=(Type=Bone,Name=\"clavicle_l\"),Weight=0.078),(Item=(Type=Bone,Name=\"deform_upperarm_fr_l\"),Weight=0.058),(Item=(Type=Bone,Name=\"upperarm_l\"),Weight=0.096),(Item=(Type=Bone,Name=\"upperarm_twist_01_l\"),Weight=0.523),(Item=(Type=Bone,Name=\"upperarm_twis ...(truncated)"
    Weight=1
   )
   > &RigUnit_ParentConstraint(
    AdvancedSettings=(InterpolationType=Average,RotationOrderForFilter=XZY)
    Child=(Type=Bone,Name="extra_arm_bag_front")
    Filter=(TranslationFilter=(bX=True,bY=True,bZ=True),RotationFilter=(bX=false,bY=false,bZ=false),ScaleFilter=(bX=True,bY=True,bZ=True))
    Parents=((Item=(Type=Bone,Name="clavicle_l"),Weight=0.065),(Item=(Type=Bone,Name="upperarm_l"),Weight=0.098),(Item=(Type=Bone,Name="upperarm_twist_01_l"),Weight=0.61),(Item=(Type=Bone,Name="upperarm_twist_02_l"),Weight=0.227))
    Weight=1
   )
   n1 = &RigVMFunction_Sequence(ExecuteContext=@RigVMFunction_Sequence_1.D)
   n2 = &RigUnit_SphericalPoseReader(
    ActiveRegionScaleFactors=(PositiveWidth=1,NegativeWidth=1.5,PositiveHeight=1,NegativeHeight=0.7)
    ActiveRegionSize=0.1
    Debug=(bDrawDebug=true,bDraw2D=true,bDrawLocalAxes=False,DebugScale=25,DebugSegments=20,DebugThickness=0.25)
    DriverAxis=(X=-1,Y=0,Z=0)
    DriverItem=(Type=Bone,Name="thigh_l")
    FalloffRegionScaleFactors=(PositiveWidth=0.5,NegativeWidth=1,PositiveHeight=0.8,NegativeHeight=0.8)
    FalloffSize=0.8
    FlipHeightScaling=False
    FlipWidthScaling=False
    OptionalParentItem=(Type=Bone,Name="pelvis")
    RotationOffset=(X=90,Y=0,Z=-60)
    ExecutePin=@n1.A
   )
   RigVMFunction_AnimEvalRichCurve = &RigVMFunction_AnimEvalRichCurve(
    Curve="(EditorCurveData=(Keys=((),(InterpMode=RCIM_Cubic,Time=0.5349,Value=0.3741,ArriveTangent=0.8529,LeaveTangent=0.8529),(InterpMode=RCIM_Cubic,Time=0.8771,Value=0.7481,ArriveTangent=1.3457,LeaveTangent=1.3457),(InterpMode=RCIM_Cubic,TangentMode=RCTM_User,Time ...(truncated)"
    SourceMinimum=0
    TargetMinimum=0
    Value=@n2.OutputParam
   )
   > /Script/ControlRig.RigUnit_SetRelativeTransformForItem(
    Child=(Type=Bone,Name="dyn_jacket_f_l_BASE")
    Parent=(Type=Bone,Name="dyn_jacket_f_l_BASE")
    Value=(Rotation=(X=0.2588,Y=-0.0002,Z=0.9659,W=0.0008),Translation=(X=-7,Y=0,Z=0),Scale3D=(X=1,Y=1,Z=1))
    bParentInitial=false
    bPropagateToChildren=true
    Weight=@RigVMFunction_AnimEvalRichCurve.Result
   ) {
    is=struct RigVMUnitNode
    template="Set Relative Transform::Execute(in Child,in Parent,in bParentInitial,in Value,in Weight,in bPropagateToChildren)"
   }
   > &RigUnit_SetRelativeTransformForItem(
    Child=(Type=Bone,Name="dyn_jacket_l_BASE")
    Parent=(Type=Bone,Name="dyn_jacket_l_BASE")
    Value=(Rotation=(X=-0.0015,Y=0.1736,Z=0.0086,W=0.9848),Translation=(X=-2,Y=2,Z=0),Scale3D=(X=1,Y=1,Z=1))
    bParentInitial=false
    bPropagateToChildren=true
    Weight=@RigVMFunction_AnimEvalRichCurve.Result
   ) {template="Set Relative Transform::Execute(in Child,in Parent,in bParentInitial,in Value,in Weight,in bPropagateToChildren)"}
   > &RigUnit_SetRelativeTransformForItem(
    Child=(Type=Bone,Name="dyn_jacket_b_l_BASE")
    Parent=(Type=Bone,Name="dyn_jacket_b_l_BASE")
    Value=(Rotation=(X=0,Y=0,Z=0.0436,W=0.999),Translation=(X=0,Y=0,Z=0),Scale3D=(X=1,Y=1,Z=1))
    bParentInitial=false
    bPropagateToChildren=true
    Weight=@RigVMFunction_AnimEvalRichCurve.Result
   ) {template="Set Relative Transform::Execute(in Child,in Parent,in bParentInitial,in Value,in Weight,in bPropagateToChildren)"}
   > &RigUnit_SetRelativeTransformForItem(
    Child=(Type=Bone,Name="dyn_card_BASE")
    Parent=(Type=Bone,Name="dyn_card_01")
    Value=(Rotation=(X=0.2143,Y=0.3659,Z=0.8284,W=0.3659),Translation=(X=-1,Y=0,Z=0),Scale3D=(X=1,Y=1,Z=1))
    bParentInitial=false
    bPropagateToChildren=true
    Weight=@RigVMFunction_AnimEvalRichCurve.Result
   ) {template="Set Relative Transform::Execute(in Child,in Parent,in bParentInitial,in Value,in Weight,in bPropagateToChildren)"}
   > n3 = &RigUnit_SphericalPoseReader(
    ActiveRegionScaleFactors=(PositiveWidth=1,NegativeWidth=1,PositiveHeight=2,NegativeHeight=2)
    ActiveRegionSize=0.1
    Debug=(bDrawDebug=false,bDraw2D=false,bDrawLocalAxes=False,DebugScale=25,DebugSegments=20,DebugThickness=0.25)
    DriverAxis=(X=-1,Y=0,Z=0)
    DriverItem=(Type=Bone,Name="thigh_l")
    FalloffRegionScaleFactors=(PositiveWidth=1,NegativeWidth=1,PositiveHeight=0.7,NegativeHeight=0.5)
    FalloffSize=0.6
    FlipHeightScaling=False
    FlipWidthScaling=False
    OptionalParentItem=(Type=Bone,Name="pelvis")
    RotationOffset=(X=180,Y=20,Z=0)
   )
   > &RigUnit_SetRelativeTransformForItem(
    Child=(Type=Bone,Name="dyn_jacket_f_l_BASE")
    Parent=(Type=Bone,Name="dyn_jacket_f_l_BASE")
    Value=(Rotation=(X=0,Y=0,Z=0,W=1),Translation=(X=0,Y=0,Z=0),Scale3D=(X=1,Y=1,Z=1))
    bParentInitial=false
    bPropagateToChildren=true
    Weight=@n3.OutputParam
   ) {template="Set Relative Transform::Execute(in Child,in Parent,in bParentInitial,in Value,in Weight,in bPropagateToChildren)"}
   > &RigUnit_SetRelativeTransformForItem(
    Child=(Type=Bone,Name="dyn_jacket_l_BASE")
    Parent=(Type=Bone,Name="dyn_jacket_l_BASE")
    Value=(Rotation=(X=0,Y=0,Z=0.887,W=0.4617),Translation=(X=0,Y=0,Z=0),Scale3D=(X=1,Y=1,Z=1))
    bParentInitial=false
    bPropagateToChildren=true
    Weight=@n3.OutputParam
   ) {template="Set Relative Transform::Execute(in Child,in Parent,in bParentInitial,in Value,in Weight,in bPropagateToChildren)"}
   > &RigUnit_SetRelativeTransformForItem(
    Child=(Type=Bone,Name="dyn_jacket_b_l_BASE")
    Parent=(Type=Bone,Name="dyn_jacket_b_l_BASE")
    Value=(Rotation=(X=0,Y=0,Z=0,W=1),Translation=(X=0,Y=0,Z=0),Scale3D=(X=1,Y=1,Z=1))
    bParentInitial=false
    bPropagateToChildren=true
    Weight=@n3.OutputParam
   ) {template="Set Relative Transform::Execute(in Child,in Parent,in bParentInitial,in Value,in Weight,in bPropagateToChildren)"}
   > &RigUnit_SetRelativeTransformForItem(
    Child=(Type=Bone,Name="dyn_card_BASE")
    Parent=(Type=Bone,Name="dyn_card_01")
    Value=(Rotation=(X=-0.2578,Y=-0.0226,Z=0.0842,W=0.9623),Translation=(X=0,Y=0,Z=0),Scale3D=(X=1,Y=1,Z=1))
    bParentInitial=false
    bPropagateToChildren=true
    Weight=@n3.OutputParam
   ) {template="Set Relative Transform::Execute(in Child,in Parent,in bParentInitial,in Value,in Weight,in bPropagateToChildren)"}
   > n4 = &RigUnit_SphericalPoseReader(
    ActiveRegionScaleFactors=(PositiveWidth=2,NegativeWidth=1,PositiveHeight=1,NegativeHeight=1)
    ActiveRegionSize=0.05
    Debug=(bDrawDebug=false,bDraw2D=false,bDrawLocalAxes=False,DebugScale=25,DebugSegments=20,DebugThickness=0.25)
    DriverAxis=(X=-1,Y=0,Z=0)
    DriverItem=(Type=Bone,Name="thigh_l")
    FalloffRegionScaleFactors=(PositiveWidth=2,NegativeWidth=5.9,PositiveHeight=2.6,NegativeHeight=2.5)
    FalloffSize=0.15
    FlipHeightScaling=False
    FlipWidthScaling=False
    OptionalParentItem=(Type=Bone,Name="pelvis")
    RotationOffset=(X=125,Y=15,Z=-30)
   )
   n5 = &RigVMFunction_AnimEvalRichCurve(
    Curve="(EditorCurveData=(Keys=((),(InterpMode=RCIM_Cubic,Time=0.1729,Value=0.2529,ArriveTangent=1.3919,LeaveTangent=1.3919),(InterpMode=RCIM_Cubic,Time=0.6096,Value=0.8485,ArriveTangent=0.9032,LeaveTangent=0.9032),(InterpMode=RCIM_Cubic,Time=1,Value=1,LeaveTangen ...(truncated)"
    SourceMinimum=0
    TargetMinimum=0
    Value=@n4.OutputParam
   )
   > &RigUnit_SetRelativeTransformForItem(
    Child=(Type=Bone,Name="dyn_jacket_f_l_BASE")
    Parent=(Type=Bone,Name="dyn_jacket_f_l_BASE")
    Value=(Rotation=(X=-0.4082,Y=-0.1094,Z=0.2346,W=0.8754),Translation=(X=0,Y=0,Z=0),Scale3D=(X=1,Y=1,Z=1))
    bParentInitial=false
    bPropagateToChildren=true
    Weight=@n5.Result
   ) {template="Set Relative Transform::Execute(in Child,in Parent,in bParentInitial,in Value,in Weight,in bPropagateToChildren)"}
   > &RigUnit_SetRelativeTransformForItem(
    Child=(Type=Bone,Name="dyn_jacket_l_BASE")
    Parent=(Type=Bone,Name="dyn_jacket_l_BASE")
    Value=(Rotation=(X=0.2578,Y=0.0226,Z=0.0842,W=0.9623),Translation=(X=0,Y=0,Z=0),Scale3D=(X=1,Y=1,Z=1))
    bParentInitial=false
    bPropagateToChildren=true
    Weight=@n5.Result
   ) {template="Set Relative Transform::Execute(in Child,in Parent,in bParentInitial,in Value,in Weight,in bPropagateToChildren)"}
   > &RigUnit_SetRelativeTransformForItem(
    Child=(Type=Bone,Name="dyn_card_BASE")
    Parent=(Type=Bone,Name="dyn_card_01")
    Value=(Rotation=(X=-0.1261,Y=-0.0338,Z=0.2566,W=0.9577),Translation=(X=0,Y=0,Z=0),Scale3D=(X=1,Y=1,Z=1))
    bParentInitial=false
    bPropagateToChildren=true
    Weight=@n5.Result
   ) {template="Set Relative Transform::Execute(in Child,in Parent,in bParentInitial,in Value,in Weight,in bPropagateToChildren)"}
   > n6 = &RigUnit_SphericalPoseReader(
    ActiveRegionScaleFactors=(PositiveWidth=1,NegativeWidth=1,PositiveHeight=1,NegativeHeight=3)
    ActiveRegionSize=0.05
    Debug=(bDrawDebug=false,bDraw2D=false,bDrawLocalAxes=False,DebugScale=25,DebugSegments=20,DebugThickness=0.25)
    DriverAxis=(X=-1,Y=0,Z=0)
    DriverItem=(Type=Bone,Name="thigh_l")
    FalloffRegionScaleFactors=(PositiveWidth=1,NegativeWidth=1.5,PositiveHeight=1,NegativeHeight=1.3)
    FalloffSize=0.4
    FlipHeightScaling=False
    FlipWidthScaling=False
    OptionalParentItem=(Type=Bone,Name="pelvis")
    RotationOffset=(X=-90,Y=0,Z=20)
   )
   n7 = &RigVMFunction_AnimEvalRichCurve(
    Curve="(EditorCurveData=(Keys=((),(InterpMode=RCIM_Cubic,Time=0.2586,Value=0.1053,ArriveTangent=0.7994,LeaveTangent=0.7994),(InterpMode=RCIM_Cubic,Time=0.806,Value=0.6443,ArriveTangent=1.2068,LeaveTangent=1.2068),(InterpMode=RCIM_Cubic,TangentMode=RCTM_User,Time= ...(truncated)"
    SourceMinimum=0
    TargetMinimum=0
    Value=@n6.OutputParam
   )
   > &RigUnit_SetRelativeTransformForItem(
    Child=(Type=Bone,Name="dyn_jacket_l_BASE")
    Parent=(Type=Bone,Name="dyn_jacket_l_BASE")
    Value=(Rotation=(X=0.0449,Y=-0.2549,Z=0.1677,W=0.9513),Translation=(X=0,Y=0,Z=0),Scale3D=(X=1,Y=1,Z=1))
    bParentInitial=false
    bPropagateToChildren=true
    Weight=@n7.Result
   ) {template="Set Relative Transform::Execute(in Child,in Parent,in bParentInitial,in Value,in Weight,in bPropagateToChildren)"}
   > &RigUnit_SetRelativeTransformForItem(
    Child=(Type=Bone,Name="dyn_jacket_b_l_BASE")
    Parent=(Type=Bone,Name="dyn_jacket_b_l_BASE")
    Value=(Rotation=(X=0,Y=0,Z=0.9063,W=0.4226),Translation=(X=-2,Y=0,Z=0),Scale3D=(X=1,Y=1,Z=1))
    bParentInitial=false
    bPropagateToChildren=true
    Weight=@n7.Result
   ) {template="Set Relative Transform::Execute(in Child,in Parent,in bParentInitial,in Value,in Weight,in bPropagateToChildren)"}
   > &RigUnit_SetRelativeTransformForItem(
    Child=(Type=Bone,Name="dyn_jacket_b_l_BASE")
    Parent=(Type=Bone,Name="dyn_jacket_b_l_BASE")
    Value=(Rotation=(X=-0.0076,Y=0.0868,Z=0.0868,W=0.9924),Translation=(X=0,Y=0,Z=0),Scale3D=(X=1,Y=1,Z=1))
    bParentInitial=false
    bPropagateToChildren=true
    Weight=&RigVMFunction_MathDoubleRemap(
     SourceMaximum=14
     SourceMinimum=0
     TargetMaximum=1
     TargetMinimum=0
     Value=&RigUnit_GetTransform(Item=(Type=Bone,Name="deform_glute_l"), Space=LocalSpace, bInitial=False).Transform.Translation.X
    ).Result
   ) {template="Set Relative Transform::Execute(in Child,in Parent,in bParentInitial,in Value,in Weight,in bPropagateToChildren)"}
   n8 = &RigUnit_SphericalPoseReader(
    ActiveRegionScaleFactors=(PositiveWidth=1,NegativeWidth=1.5,PositiveHeight=1,NegativeHeight=0.7)
    ActiveRegionSize=0.1
    Debug=(bDrawDebug=True,bDraw2D=true,bDrawLocalAxes=False,DebugScale=25,DebugSegments=20,DebugThickness=0.25)
    DriverAxis=(X=-1,Y=0,Z=0)
    DriverItem=(Type=Bone,Name="thigh_r")
    FalloffRegionScaleFactors=(PositiveWidth=0.5,NegativeWidth=1,PositiveHeight=0.8,NegativeHeight=0.8)
    FalloffSize=0.8
    FlipHeightScaling=False
    FlipWidthScaling=False
    OptionalParentItem=(Type=Bone,Name="pelvis")
    RotationOffset=(X=90,Y=0,Z=-60)
    ExecutePin=@n1.B
   )
   n9 = &RigVMFunction_AnimEvalRichCurve(
    Curve="(EditorCurveData=(Keys=((),(InterpMode=RCIM_Cubic,Time=0.5333,Value=0.411,ArriveTangent=0.8155,LeaveTangent=0.8155),(InterpMode=RCIM_Cubic,Time=0.8817,Value=0.719,ArriveTangent=1.2622,LeaveTangent=1.2622),(InterpMode=RCIM_Cubic,TangentMode=RCTM_User,Time=1 ...(truncated)"
    SourceMinimum=0
    TargetMinimum=0
    Value=@n8.OutputParam
   )
   > &RigUnit_SetRelativeTransformForItem(
    Child=(Type=Bone,Name="dyn_jacket_f_r_BASE")
    Parent=(Type=Bone,Name="dyn_jacket_f_r_BASE")
    Value=(Rotation=(X=0.0151,Y=0.173,Z=0.9811,W=0.0858),Translation=(X=10,Y=0,Z=0),Scale3D=(X=1,Y=1,Z=1))
    bParentInitial=false
    bPropagateToChildren=true
    Weight=@n9.Result
   ) {template="Set Relative Transform::Execute(in Child,in Parent,in bParentInitial,in Value,in Weight,in bPropagateToChildren)"}
   > &RigUnit_SetRelativeTransformForItem(
    Child=(Type=Bone,Name="dyn_jacket_r_BASE")
    Parent=(Type=Bone,Name="dyn_jacket_r_BASE")
    Value=(Rotation=(X=0.0015,Y=-0.1736,Z=0.0086,W=0.9848),Translation=(X=2,Y=-2,Z=0),Scale3D=(X=1,Y=1,Z=1))
    bParentInitial=false
    bPropagateToChildren=true
    Weight=@n9.Result
   ) {template="Set Relative Transform::Execute(in Child,in Parent,in bParentInitial,in Value,in Weight,in bPropagateToChildren)"}
   > &RigUnit_SetRelativeTransformForItem(
    Child=(Type=Bone,Name="dyn_jacket_b_r_BASE1")
    Parent=(Type=Bone,Name="dyn_jacket_b_r_BASE1")
    Value=(Rotation=(X=0,Y=0,Z=0.0436,W=0.999),Translation=(X=0,Y=0,Z=0),Scale3D=(X=1,Y=1,Z=1))
    bParentInitial=false
    bPropagateToChildren=true
    Weight=@n9.Result
   ) {template="Set Relative Transform::Execute(in Child,in Parent,in bParentInitial,in Value,in Weight,in bPropagateToChildren)"}
   > n10 = &RigUnit_SphericalPoseReader(
    ActiveRegionScaleFactors=(PositiveWidth=1,NegativeWidth=1,PositiveHeight=2,NegativeHeight=2)
    ActiveRegionSize=0.1
    Debug=(bDrawDebug=True,bDraw2D=true,bDrawLocalAxes=False,DebugScale=25,DebugSegments=20,DebugThickness=0.25)
    DriverAxis=(X=-1,Y=0,Z=0)
    DriverItem=(Type=Bone,Name="thigh_r")
    FalloffRegionScaleFactors=(PositiveWidth=1,NegativeWidth=1,PositiveHeight=0.7,NegativeHeight=0.5)
    FalloffSize=0.6
    FlipHeightScaling=False
    FlipWidthScaling=False
    OptionalParentItem=(Type=Bone,Name="pelvis")
    RotationOffset=(X=180,Y=20,Z=0)
   )
   > &RigUnit_SetRelativeTransformForItem(
    Child=(Type=Bone,Name="dyn_jacket_f_r_BASE")
    Parent=(Type=Bone,Name="dyn_jacket_f_r_BASE")
    Value=(Rotation=(X=0,Y=0,Z=0,W=1),Translation=(X=0,Y=0,Z=0),Scale3D=(X=1,Y=1,Z=1))
    bParentInitial=false
    bPropagateToChildren=true
    Weight=@n10.OutputParam
   ) {template="Set Relative Transform::Execute(in Child,in Parent,in bParentInitial,in Value,in Weight,in bPropagateToChildren)"}
   > &RigUnit_SetRelativeTransformForItem(
    Child=(Type=Bone,Name="dyn_jacket_r_BASE")
    Parent=(Type=Bone,Name="dyn_jacket_r_BASE")
    Value=(Rotation=(X=0,Y=0,Z=0.887,W=0.4617),Translation=(X=2,Y=0,Z=0),Scale3D=(X=1,Y=1,Z=1))
    bParentInitial=false
    bPropagateToChildren=true
    Weight=@n10.OutputParam
   ) {template="Set Relative Transform::Execute(in Child,in Parent,in bParentInitial,in Value,in Weight,in bPropagateToChildren)"}
   > &RigUnit_SetRelativeTransformForItem(
    Child=(Type=Bone,Name="dyn_jacket_b_l_BASE")
    Parent=(Type=Bone,Name="dyn_jacket_b_l_BASE")
    Value=(Rotation=(X=0,Y=0,Z=0,W=1),Translation=(X=0,Y=0,Z=0),Scale3D=(X=1,Y=1,Z=1))
    bParentInitial=false
    bPropagateToChildren=true
    Weight=@n10.OutputParam
   ) {template="Set Relative Transform::Execute(in Child,in Parent,in bParentInitial,in Value,in Weight,in bPropagateToChildren)"}
   > &RigUnit_SetRelativeTransformForItem(
    Child=(Type=Bone,Name="dyn_katana_BASE")
    Parent=(Type=Bone,Name="dyn_katana_BASE")
    Value=(Rotation=(X=0,Y=0,Z=0,W=1),Translation=(X=0,Y=-3,Z=0),Scale3D=(X=1,Y=1,Z=1))
    bParentInitial=false
    bPropagateToChildren=true
    Weight=@n10.OutputParam
   ) {template="Set Relative Transform::Execute(in Child,in Parent,in bParentInitial,in Value,in Weight,in bPropagateToChildren)"}
   > n11 = &RigUnit_SphericalPoseReader(
    ActiveRegionScaleFactors=(PositiveWidth=2,NegativeWidth=1,PositiveHeight=1,NegativeHeight=1)
    ActiveRegionSize=0.05
    Debug=(bDrawDebug=True,bDraw2D=true,bDrawLocalAxes=False,DebugScale=25,DebugSegments=20,DebugThickness=0.25)
    DriverAxis=(X=-1,Y=0,Z=0)
    DriverItem=(Type=Bone,Name="thigh_r")
    FalloffRegionScaleFactors=(PositiveWidth=2,NegativeWidth=5.9,PositiveHeight=2.6,NegativeHeight=2.7)
    FalloffSize=0.15
    FlipHeightScaling=False
    FlipWidthScaling=False
    OptionalParentItem=(Type=Bone,Name="pelvis")
    RotationOffset=(X=125,Y=15,Z=-30)
   )
   n12 = &RigVMFunction_AnimEvalRichCurve(
    Curve="(EditorCurveData=(Keys=((),(InterpMode=RCIM_Cubic,Time=0.1729,Value=0.2529,ArriveTangent=1.5337,LeaveTangent=1.5337),(InterpMode=RCIM_Cubic,Time=0.61,Value=0.9356,ArriveTangent=0.9032,LeaveTangent=0.9032),(InterpMode=RCIM_Cubic,Time=1,Value=1,LeaveTangent= ...(truncated)"
    SourceMinimum=0
    TargetMinimum=0
    Value=@n11.OutputParam
   )
   > &RigUnit_SetRelativeTransformForItem(
    Child=(Type=Bone,Name="dyn_jacket_f_r_BASE")
    Parent=(Type=Bone,Name="dyn_jacket_f_r_BASE")
    Value=(Rotation=(X=-0.3971,Y=-0.1445,Z=0.31,W=0.8517),Translation=(X=0,Y=0,Z=0),Scale3D=(X=1,Y=1,Z=1))
    bParentInitial=false
    bPropagateToChildren=true
    Weight=@n12.Result
   ) {template="Set Relative Transform::Execute(in Child,in Parent,in bParentInitial,in Value,in Weight,in bPropagateToChildren)"}
   > &RigUnit_SetRelativeTransformForItem(
    Child=(Type=Bone,Name="dyn_jacket_r_BASE")
    Parent=(Type=Bone,Name="dyn_jacket_r_BASE")
    Value=(Rotation=(X=0.2549,Y=0.0449,Z=0.1677,W=0.9513),Translation=(X=0,Y=0,Z=0),Scale3D=(X=1,Y=1,Z=1))
    bParentInitial=false
    bPropagateToChildren=true
    Weight=@n12.Result
   ) {template="Set Relative Transform::Execute(in Child,in Parent,in bParentInitial,in Value,in Weight,in bPropagateToChildren)"}
   > &RigUnit_SetRelativeTransformForItem(
    Child=(Type=Bone,Name="dyn_katana_BASE")
    Parent=(Type=Bone,Name="dyn_katana_BASE")
    Value=(Rotation=(X=0,Y=0,Z=0,W=1),Translation=(X=0,Y=-2,Z=0),Scale3D=(X=1,Y=1,Z=1))
    bParentInitial=false
    bPropagateToChildren=true
    Weight=@n12.Result
   ) {template="Set Relative Transform::Execute(in Child,in Parent,in bParentInitial,in Value,in Weight,in bPropagateToChildren)"}
   > n13 = &RigUnit_SphericalPoseReader(
    ActiveRegionScaleFactors=(PositiveWidth=1,NegativeWidth=1,PositiveHeight=1,NegativeHeight=3)
    ActiveRegionSize=0.05
    Debug=(bDrawDebug=True,bDraw2D=true,bDrawLocalAxes=False,DebugScale=25,DebugSegments=20,DebugThickness=0.25)
    DriverAxis=(X=-1,Y=0,Z=0)
    DriverItem=(Type=Bone,Name="thigh_r")
    FalloffRegionScaleFactors=(PositiveWidth=1,NegativeWidth=1.5,PositiveHeight=1,NegativeHeight=1.3)
    FalloffSize=0.4
    FlipHeightScaling=False
    FlipWidthScaling=False
    OptionalParentItem=(Type=Bone,Name="pelvis")
    RotationOffset=(X=-90,Y=0,Z=20)
   )
   n14 = &RigVMFunction_AnimEvalRichCurve(
    Curve="(EditorCurveData=(Keys=((),(InterpMode=RCIM_Cubic,Time=0.2586,Value=0.1053,ArriveTangent=0.7994,LeaveTangent=0.7994),(InterpMode=RCIM_Cubic,Time=0.806,Value=0.6443,ArriveTangent=1.2068,LeaveTangent=1.2068),(InterpMode=RCIM_Cubic,TangentMode=RCTM_User,Time= ...(truncated)"
    SourceMinimum=0
    TargetMinimum=0
    Value=@n13.OutputParam
   )
   > &RigUnit_SetRelativeTransformForItem(
    Child=(Type=Bone,Name="dyn_jacket_r_BASE")
    Parent=(Type=Bone,Name="dyn_jacket_r_BASE")
    Value=(Rotation=(X=0.0449,Y=-0.2549,Z=0.1677,W=0.9513),Translation=(X=0,Y=0,Z=0),Scale3D=(X=1,Y=1,Z=1))
    bParentInitial=false
    bPropagateToChildren=true
    Weight=@n14.Result
   ) {template="Set Relative Transform::Execute(in Child,in Parent,in bParentInitial,in Value,in Weight,in bPropagateToChildren)"}
   > &RigUnit_SetRelativeTransformForItem(
    Child=(Type=Bone,Name="dyn_jacket_b_r_BASE1")
    Parent=(Type=Bone,Name="dyn_jacket_b_r_BASE1")
    Value=(Rotation=(X=0,Y=0,Z=0.9063,W=0.4226),Translation=(X=1,Y=0,Z=0),Scale3D=(X=1,Y=1,Z=1))
    bParentInitial=false
    bPropagateToChildren=true
    Weight=@n14.Result
   ) {template="Set Relative Transform::Execute(in Child,in Parent,in bParentInitial,in Value,in Weight,in bPropagateToChildren)"}
   > &RigUnit_SetRelativeTransformForItem(
    Child=(Type=Bone,Name="dyn_katana_BASE")
    Parent=(Type=Bone,Name="dyn_katana_BASE")
    Value=(Rotation=(X=0,Y=0.0872,Z=0,W=0.9962),Translation=(X=0,Y=-4,Z=0),Scale3D=(X=1,Y=1,Z=1))
    bParentInitial=false
    bPropagateToChildren=true
    Weight=@n14.Result
   ) {template="Set Relative Transform::Execute(in Child,in Parent,in bParentInitial,in Value,in Weight,in bPropagateToChildren)"}
   > &RigUnit_SetRelativeTransformForItem(
    Child=(Type=Bone,Name="extra_pelvis_bag_back")
    Parent=(Type=Bone,Name="extra_pelvis_bag_back")
    Value=(Rotation=(X=0,Y=0,Z=0,W=1),Translation=(X=-2,Y=0,Z=0),Scale3D=(X=1,Y=1,Z=1))
    bParentInitial=false
    bPropagateToChildren=true
    Weight=@n14.Result
   ) {template="Set Relative Transform::Execute(in Child,in Parent,in bParentInitial,in Value,in Weight,in bPropagateToChildren)"}
   > &RigUnit_SetRelativeTransformForItem(
    Child=(Type=Bone,Name="dyn_jacket_b_r_BASE1")
    Parent=(Type=Bone,Name="dyn_jacket_b_r_BASE1")
    Value=(Rotation=(X=-0.0076,Y=0.0868,Z=0.0868,W=0.9924),Translation=(X=0,Y=0,Z=0),Scale3D=(X=1,Y=1,Z=1))
    bParentInitial=false
    bPropagateToChildren=true
    Weight=&RigVMFunction_MathDoubleRemap(
     SourceMaximum=14
     SourceMinimum=0
     TargetMaximum=1
     TargetMinimum=0
     Value=&RigUnit_GetTransform(Item=(Type=Bone,Name="deform_glute_r"), Space=LocalSpace, bInitial=False).Transform.Translation.X
    ).Result
   ) {template="Set Relative Transform::Execute(in Child,in Parent,in bParentInitial,in Value,in Weight,in bPropagateToChildren)"}
   n15 = &RigVMFunction_MathDoubleRemap(
    SourceMaximum=70
    SourceMinimum=8
    TargetMaximum=1
    TargetMinimum=0
    Value=&RigVMFunction_MathQuaternionToEuler(
     RotationOrder=YXZ
     Value=&RigUnit_GetTransform(Item=(Type=Bone,Name="spine_03"), Space=LocalSpace, bInitial=False).Transform.Rotation
    ).Result.Z
   )
   &RigUnit_SetRelativeTransformForItem(
    Child=(Type=Bone,Name="extra_chest_bag_bot")
    Parent=(Type=Bone,Name="extra_chest_bag_bot")
    Value=(Rotation=(X=0,Y=0,Z=0.3827,W=0.9239),Translation=(X=2,Y=1,Z=0),Scale3D=(X=1,Y=1,Z=1))
    bParentInitial=false
    bPropagateToChildren=true
    ExecutePin=@RigVMFunction_Sequence_1.C
    Weight=@n15.Result
   ) {template="Set Relative Transform::Execute(in Child,in Parent,in bParentInitial,in Value,in Weight,in bPropagateToChildren)"}
   > &RigUnit_SetRelativeTransformForItem(
    Child=(Type=Bone,Name="extra_chest_bag_top")
    Parent=(Type=Bone,Name="extra_chest_bag_top")
    Value=(Rotation=(X=0,Y=0,Z=0.0785,W=0.9969),Translation=(X=-4,Y=0,Z=0),Scale3D=(X=1,Y=1,Z=1))
    bParentInitial=false
    bPropagateToChildren=true
    Weight=@n15.Result
   ) {template="Set Relative Transform::Execute(in Child,in Parent,in bParentInitial,in Value,in Weight,in bPropagateToChildren)"}
   > &RigUnit_SetRelativeTransformForItem(
    Child=(Type=Bone,Name="extra_cup1")
    Parent=(Type=Bone,Name="extra_cup1")
    Value=(Rotation=(X=0,Y=0,Z=0.0785,W=0.9969),Translation=(X=-4,Y=0,Z=0),Scale3D=(X=1,Y=1,Z=1))
    bParentInitial=false
    bPropagateToChildren=true
    Weight=@n15.Result
   ) {template="Set Relative Transform::Execute(in Child,in Parent,in bParentInitial,in Value,in Weight,in bPropagateToChildren)"}
   n16 = &RigVMFunction_MathDoubleRemap(
    SourceMaximum=60
    SourceMinimum=0
    TargetMaximum=1
    TargetMinimum=0
    Value=&RigVMFunction_MathQuaternionToEuler(
     RotationOrder=YXZ
     Value=&RigUnit_GetTransform(Item=(Type=Bone,Name="spine_04"), Space=LocalSpace, bInitial=False).Transform.Rotation
    ).Result.Z
   )
   > &RigUnit_SetRelativeTransformForItem(
    Child=(Type=Bone,Name="extra_chest_bag_bot")
    Parent=(Type=Bone,Name="extra_chest_bag_bot")
    Value=(Rotation=(X=0,Y=0,Z=0.1045,W=0.9945),Translation=(X=4,Y=0.5,Z=0),Scale3D=(X=1,Y=1,Z=1))
    bParentInitial=false
    bPropagateToChildren=true
    Weight=@n16.Result
   ) {template="Set Relative Transform::Execute(in Child,in Parent,in bParentInitial,in Value,in Weight,in bPropagateToChildren)"}
   > &RigUnit_SetRelativeTransformForItem(
    Child=(Type=Bone,Name="extra_chest_bag_top")
    Parent=(Type=Bone,Name="extra_chest_bag_top")
    Value=(Rotation=(X=0,Y=0,Z=0.0785,W=0.9969),Translation=(X=-4,Y=2,Z=0),Scale3D=(X=1,Y=1,Z=1))
    bParentInitial=false
    bPropagateToChildren=true
    Weight=@n16.Result
   ) {template="Set Relative Transform::Execute(in Child,in Parent,in bParentInitial,in Value,in Weight,in bPropagateToChildren)"}
   > &RigUnit_SetRelativeTransformForItem(
    Child=(Type=Bone,Name="extra_cup1")
    Parent=(Type=Bone,Name="extra_cup1")
    Value=(Rotation=(X=0,Y=0,Z=0.0785,W=0.9969),Translation=(X=-5,Y=0.5,Z=0),Scale3D=(X=1,Y=1,Z=1))
    bParentInitial=false
    bPropagateToChildren=true
    Weight=@n16.Result
   ) {template="Set Relative Transform::Execute(in Child,in Parent,in bParentInitial,in Value,in Weight,in bPropagateToChildren)"}
  ]}
]
hierarchy [
 root {loc=<0 0 0>, rot=<0 0 0 1>}
 pelvis {parent=root, loc=<0 -0.7071 93.1191>, rot=<0.0127 0.707 -0.0127 -0.707>}
 spine_01 {parent=pelvis, loc=<6.7494 -0.6388 0>, rot=<0 0 0.034 -0.9994>}
 spine_02 {parent=spine_01, loc=<7.2518 -0.035 0>, rot=<0 0 0.0238 -0.9997>}
 spine_03 {parent=spine_02, loc=<6.0104 0 0>, rot=<0 0 0.0149 -0.9999>}
 spine_04 {parent=spine_03, loc=<8.1024 0 0>, rot=<0 0 -0.0132 -0.9999>}
 spine_05 {parent=spine_04, loc=<10.0146 0 0>, rot=<0 0 -0.0015 -1>}
 neck_01 {parent=spine_05, loc=<11.3138 0.1529 0>, rot=<0 0 -0.2003 -0.9797>}
 neck_02 {parent=neck_01, loc=<4.5627 0 0>, rot=<0 0 0.0504 -0.9987>}
 head {parent=neck_02, loc=<5.3914 0 0>, rot=<0 0 0.115 -0.9934>}
 FACIAL_C_FacialRoot {parent=head, loc=<2.6735 1.1079 0>, rot=<0.5021 -0.4979 -0.5021 -0.4979>}
 FACIAL_L_Masseter {parent=FACIAL_C_FacialRoot, loc=<5.8306 1.1423 3.0841>, rot=<0.0693 0.1129 0.0079 -0.9912>}
 FACIAL_R_Masseter {parent=FACIAL_C_FacialRoot, loc=<-5.8306 1.1423 3.0841>, rot=<0.0693 -0.1129 -0.0079 -0.9912>}
 FACIAL_C_UnderChin {parent=FACIAL_C_FacialRoot, loc=<0 8.3871 4.2711>, rot=<0.5 0 0 -0.866>}
 FACIAL_L_UnderChin {parent=FACIAL_C_FacialRoot, loc=<3.1079 6.9697 2.6755>, rot=<0.4025 -0.355 0.0823 -0.8398>}
 FACIAL_R_UnderChin {parent=FACIAL_C_FacialRoot, loc=<-3.1076 6.9697 2.6763>, rot=<0.4025 0.355 -0.0823 -0.8398>}
 FACIAL_C_Jaw {parent=FACIAL_C_FacialRoot, loc=<0 1.965 1.3036>, rot=<0.342 0 0 -0.9397>}
 FACIAL_C_Jawline {parent=FACIAL_C_Jaw, loc=<0 1.8765 7.8411>, rot=<0.2588 0 0 -0.9659>}
 FACIAL_L_ChinSide {parent=FACIAL_C_Jaw, loc=<2.5905 -0.0165 7.1179>, rot=<-0.2207 -0.3739 0.05 -0.8994>}
 FACIAL_R_ChinSide {parent=FACIAL_C_Jaw, loc=<-2.5905 -0.0165 7.1179>, rot=<-0.2207 0.3739 -0.05 -0.8994>}
 FACIAL_C_Chin {parent=FACIAL_C_Jaw, loc=<0 0.447 7.5452>, rot=<-0.2588 0 0 -0.9659>}
 FACIAL_L_Ear {parent=FACIAL_C_FacialRoot, loc=<6.1345 -0.2266 0.3932>, rot=<-0.0262 -0.0007 -0.0262 -0.9993>}
 FACIAL_R_Ear {parent=FACIAL_C_FacialRoot, loc=<-6.1345 -0.2266 0.3932>, rot=<-0.0262 0.0007 0.0262 -0.9993>}
 FACIAL_L_Jawline {parent=FACIAL_C_FacialRoot, loc=<4.2329 4.7612 3.5675>, rot=<0.2092 0.1578 -0.0547 -0.9635>}
 FACIAL_R_Jawline {parent=FACIAL_C_FacialRoot, loc=<-4.2329 4.7612 3.5675>, rot=<0.2092 -0.1578 0.0547 -0.9635>}
 attach_cape {parent=spine_05, loc=<9.017 -11.714 0>, rot=<-0.3248 -0.5876 0.1669 -0.722>}
 attach_backpack {parent=spine_05, loc=<-0.123 -13.473 0>, rot=<-0.0093 -0.707 0.0093 -0.707>}
 clavicle_l {parent=spine_05, loc=<4.3575 -0.1466 -2.3223>, rot=<0.0465 -0.7289 -0.009 -0.6829>}
 upperarm_l {parent=clavicle_l, loc=<15.5111 0 0>, rot=<0.0226 -0.3234 -0.0088 -0.9459>}
 lowerarm_l {parent=upperarm_l, loc=<26.6514 0 0>, rot=<0.0391 -0.0099 -0.2453 -0.9686>}
 hand_l {parent=lowerarm_l, loc=<22.3491 0 0>, rot=<-0.7081 -0.0507 0.0975 -0.6975>}
 thumb_01_l {parent=hand_l, loc=<1.3118 -0.3191 -2.459>, rot=<0.6546 -0.4042 0.0664 -0.6354>}
 thumb_02_l {parent=thumb_01_l, loc=<5.9336 -0.0448 0.0401>, rot=<-0.0037 0.0213 0.298 -0.9543>}
 thumb_03_l {parent=thumb_02_l, loc=<2.2223 0.3682 -0.0907>, rot=<0.0118 -0.013 0.0066 -0.9998>}
 index_metacarpal_l {parent=hand_l, loc=<1.6535 0.0183 -2.96>, rot=<0.0158 -0.0124 -0.0173 -0.9996>}
 index_01_l {parent=index_metacarpal_l, loc=<8.3008 0.0357 -0.0524>, rot=<-0.0044 0.0485 0.0924 -0.9945>}
 index_02_l {parent=index_01_l, loc=<3.3587 -0.3004 0.0452>, rot=<-0.002 0.0027 0.3225 -0.9466>}
 index_03_l {parent=index_02_l, loc=<2.5578 -0.123 -0.0252>, rot=<-0.002 0.0067 0.2558 -0.9667>}
 middle_metacarpal_l {parent=hand_l, loc=<2.1856 -0.1725 -0.5984>, rot=<0.006 0.0168 -0.0329 -0.9993>}
 middle_01_l {parent=middle_metacarpal_l, loc=<7.3275 -0.2447 0.0129>, rot=<-0.0208 0.0042 0.0858 -0.9961>}
 middle_02_l {parent=middle_01_l, loc=<4.0123 0.1401 0.1175>, rot=<-0.0218 0.0258 0.3512 -0.9357>}
 middle_03_l {parent=middle_02_l, loc=<2.6077 0.0534 0.1373>, rot=<0.0082 0.0188 0.1708 -0.9851>}
 ring_metacarpal_l {parent=hand_l, loc=<2.2074 -0.3625 1.3638>, rot=<-0.0284 0.0722 -0.0458 -0.9959>}
 ring_01_l {parent=ring_metacarpal_l, loc=<7.0295 -0.6197 -0.1446>, rot=<-0.0489 -0.0252 0.1892 -0.9804>}
 ring_02_l {parent=ring_01_l, loc=<3.6174 0.4647 -0.1027>, rot=<0 0 0.2806 -0.9598>}
 ring_03_l {parent=ring_02_l, loc=<2.4152 0.114 0.0199>, rot=<0 0 0.1189 -0.9929>}
 pinky_metacarpal_l {parent=hand_l, loc=<1.7616 -0.9753 3.3814>, rot=<-0.0397 0.0814 -0.069 -0.9935>}
 pinky_01_l {parent=pinky_metacarpal_l, loc=<7.1709 -0.4831 -0.1094>, rot=<-0.0654 -0.0035 0.2242 -0.9723>}
 pinky_02_l {parent=pinky_01_l, loc=<2.8274 0.0893 -0.1043>, rot=<0.0037 0.0011 0.2854 -0.9584>}
 pinky_03_l {parent=pinky_02_l, loc=<2.2046 0.0307 -0.0078>, rot=<-0.0002 0.0098 0.1445 -0.9895>}
 deform_wrist_dn_base_l {parent=hand_l, loc=<0 -2.2342 0>, rot=<0 0 0 -1>}
 deform_wrist_dn_l {parent=deform_wrist_dn_base_l, loc=<0 0 0>, rot=<0 0 0 1>}
 deform_wrist_up_base_l {parent=hand_l, loc=<0 2.2682 0>, rot=<0 0 0 -1>}
 deform_wrist_up_l {parent=deform_wrist_up_base_l, loc=<0 0 0>, rot=<0 0 0 1>}
 lowerarm_twist_02_l {parent=lowerarm_l, loc=<9.6199 0 0>, rot=<0 0 0 -1>}
 lowerarm_twist_01_l {parent=lowerarm_l, loc=<17.8549 0 0>, rot=<0 0 0 -1>}
 deform_elbow_in_base_l {parent=lowerarm_l, loc=<0.2112 5.259 0.8274>, rot=<0 0 0 -1>}
 deform_elbow_in_l {parent=deform_elbow_in_base_l, loc=<0 0 0>, rot=<0 0 0 1>}
 deform_elbow_out_base_l {parent=lowerarm_l, loc=<0 -4 0>, rot=<0 0 0 -1>}
 deform_elbow_out_l {parent=deform_elbow_out_base_l, loc=<0 0 0>, rot=<0 0 0 1>}
 upperarm_twist_01_l {parent=upperarm_l, loc=<0.0041 0 0>, rot=<0 0 0 1>}
 deform_upperarm_bk_base_l {parent=upperarm_twist_01_l, loc=<1.0813 -7.2193 -1.9453>, rot=<0 0 0 1>}
 deform_upperarm_bk_l {parent=deform_upperarm_bk_base_l, loc=<0 0 0>, rot=<0 0 0 1>}
 deform_upperarm_fr_base_l {parent=upperarm_twist_01_l, loc=<-0.3691 7.0506 -2.765>, rot=<0 0 0 1>}
 deform_upperarm_fr_l {parent=deform_upperarm_fr_base_l, loc=<0 0 0>, rot=<0 0 0 1>}
 deform_upperarm_up_base_l {parent=upperarm_twist_01_l, loc=<-3.2216 -1.1391 3.3311>, rot=<0 0 0 1>}
 deform_upperarm_up_l {parent=deform_upperarm_up_base_l, loc=<0 0 0>, rot=<0 0 0 1>}
 extra_arm_bag_front {parent=upperarm_twist_01_l, loc=<5.1351 1.142 4.7241>, rot=<-0.577 -0.0325 0.0297 -0.8156>}
 extra_arm_bag_back {parent=upperarm_twist_01_l, loc=<5.1381 -3.2606 4.665>, rot=<0.5592 -0.0607 -0.0374 -0.826>}
 upperarm_twist_02_l {parent=upperarm_l, loc=<13.7941 0 0>, rot=<0 0 0 1>}
 clavicle_r {parent=spine_05, loc=<4.3575 -0.1466 2.3223>, rot=<0.7289 0.0465 -0.6829 0.009>}
 upperarm_r {parent=clavicle_r, loc=<-15.5111 0 0>, rot=<0.0226 -0.3234 -0.0088 -0.9459>}
 lowerarm_r {parent=upperarm_r, loc=<-26.6514 0 0>, rot=<0.0391 -0.0099 -0.2453 -0.9686>}
 hand_r {parent=lowerarm_r, loc=<-22.3491 0 0>, rot=<-0.7081 -0.0507 0.0975 -0.6975>}
 thumb_01_r {parent=hand_r, loc=<-1.3118 0.3191 2.459>, rot=<0.6546 -0.4042 0.0664 -0.6354>}
 thumb_02_r {parent=thumb_01_r, loc=<-5.9336 0.0448 -0.0401>, rot=<-0.0037 0.0213 0.298 -0.9543>}
 thumb_03_r {parent=thumb_02_r, loc=<-2.2223 -0.3682 0.0907>, rot=<0.0118 -0.013 0.0066 -0.9998>}
 index_metacarpal_r {parent=hand_r, loc=<-1.6535 -0.0183 2.96>, rot=<0.0158 -0.0124 -0.0173 -0.9996>}
 index_01_r {parent=index_metacarpal_r, loc=<-8.3008 -0.0357 0.0524>, rot=<-0.0044 0.0485 0.0924 -0.9945>}
 index_02_r {parent=index_01_r, loc=<-3.3587 0.3004 -0.0452>, rot=<-0.002 0.0027 0.3225 -0.9466>}
 index_03_r {parent=index_02_r, loc=<-2.5578 0.123 0.0252>, rot=<-0.002 0.0067 0.2558 -0.9667>}
 middle_metacarpal_r {parent=hand_r, loc=<-2.1856 0.1725 0.5984>, rot=<0.006 0.0168 -0.0329 -0.9993>}
 middle_01_r {parent=middle_metacarpal_r, loc=<-7.3275 0.2447 -0.0129>, rot=<-0.0208 0.0042 0.0858 -0.9961>}
 middle_02_r {parent=middle_01_r, loc=<-4.0123 -0.1401 -0.1175>, rot=<-0.0218 0.0258 0.3512 -0.9357>}
 middle_03_r {parent=middle_02_r, loc=<-2.6077 -0.0534 -0.1373>, rot=<0.0082 0.0188 0.1708 -0.9851>}
 ring_metacarpal_r {parent=hand_r, loc=<-2.2074 0.3625 -1.3638>, rot=<-0.0284 0.0722 -0.0458 -0.9959>}
 ring_01_r {parent=ring_metacarpal_r, loc=<-7.0295 0.6197 0.1446>, rot=<-0.0489 -0.0252 0.1892 -0.9804>}
 ring_02_r {parent=ring_01_r, loc=<-3.6174 -0.4647 0.1027>, rot=<0 0 0.2806 -0.9598>}
 ring_03_r {parent=ring_02_r, loc=<-2.4152 -0.114 -0.0199>, rot=<0 0 0.1189 -0.9929>}
 pinky_metacarpal_r {parent=hand_r, loc=<-1.7616 0.9753 -3.3814>, rot=<-0.0397 0.0814 -0.069 -0.9935>}
 pinky_01_r {parent=pinky_metacarpal_r, loc=<-7.1709 0.4831 0.1094>, rot=<-0.0654 -0.0035 0.2242 -0.9723>}
 pinky_02_r {parent=pinky_01_r, loc=<-2.8274 -0.0893 0.1043>, rot=<0.0037 0.0011 0.2854 -0.9584>}
 pinky_03_r {parent=pinky_02_r, loc=<-2.2046 -0.0307 0.0078>, rot=<-0.0002 0.0098 0.1445 -0.9895>}
 deform_wrist_dn_base_r {parent=hand_r, loc=<-0.0004 2.2345 -0.0001>, rot=<0 0 0 -1>}
 deform_wrist_dn_r {parent=deform_wrist_dn_base_r, loc=<0 0 0>, rot=<0 0 0 1>}
 deform_wrist_up_base_r {parent=hand_r, loc=<0.0003 -2.2684 0.0001>, rot=<0 0 0 -1>}
 deform_wrist_up_r {parent=deform_wrist_up_base_r, loc=<0 0 0>, rot=<0 0 0 1>}
 lowerarm_twist_02_r {parent=lowerarm_r, loc=<-9.6199 0 0>, rot=<0 0 0 -1>}
 lowerarm_twist_01_r {parent=lowerarm_r, loc=<-17.8549 0 0>, rot=<0 0 0 -1>}
 deform_elbow_in_base_r {parent=lowerarm_r, loc=<-0.2111 -5.2591 -0.8275>, rot=<0 0 0 -1>}
 deform_elbow_in_r {parent=deform_elbow_in_base_r, loc=<-0.0001 0 0.0002>, rot=<0 0 0 1>}
 deform_elbow_out_base_r {parent=lowerarm_r, loc=<0 4 -0.0001>, rot=<0 0 0 -1>}
 deform_elbow_out_r {parent=deform_elbow_out_base_r, loc=<0 0 0>, rot=<0 0 0 1>}
 upperarm_twist_01_r {parent=upperarm_r, loc=<-0.0041 0 0>, rot=<0 0 0 -1>}
 deform_upperarm_bk_base_r {parent=upperarm_twist_01_r, loc=<-1.0814 7.2192 1.9455>, rot=<0 0 0 -1>}
 deform_upperarm_bk_r {parent=deform_upperarm_bk_base_r, loc=<0 0 0>, rot=<0 0 0 1>}
 deform_upperarm_fr_base_r {parent=upperarm_twist_01_r, loc=<0.3689 -7.0506 2.7652>, rot=<0 0 0 -1>}
 deform_upperarm_fr_r {parent=deform_upperarm_fr_base_r, loc=<0 0 0>, rot=<0 0 0 1>}
 deform_upperarm_up_base_r {parent=upperarm_twist_01_r, loc=<3.2214 1.1391 -3.3308>, rot=<0 0 0 1>}
 deform_upperarm_up_r {parent=deform_upperarm_up_base_r, loc=<0 0 0>, rot=<0 0 0 -1>}
 upperarm_twist_02_r {parent=upperarm_r, loc=<-13.7941 0 0>, rot=<0 0 0 1>}
 deform_pec_aim_base_l {parent=spine_04, loc=<7.8732 12.6388 -2.4123>, rot=<0.0102 -0.4602 0.3979 -0.7936>}
 deform_pec_aim_l {parent=deform_pec_aim_base_l, loc=<0 0 0>, rot=<0 0 0 1>}
 deform_pec_base_l {parent=deform_pec_aim_l, loc=<10.766 0 0>, rot=<0 0 0 1>}
 deform_pec_l {parent=deform_pec_base_l, loc=<0.055 0 0>, rot=<0 0 0 1>}
 deform_pec_aim_base_r {parent=spine_04, loc=<7.8727 12.6388 2.4123>, rot=<-0.4602 -0.0102 0.7936 0.3979>}
 deform_pec_aim_r {parent=deform_pec_aim_base_r, loc=<0 0 0>, rot=<0 0 0 1>}
 deform_pec_base_r {parent=deform_pec_aim_r, loc=<-10.7662 -0.0006 -0.0007>, rot=<0 0 0 1>}
 deform_pec_r {parent=deform_pec_base_r, loc=<0 0 0>, rot=<0 0 0 1>}
 extra_chest_bag_top {parent=spine_04, loc=<4.6693 12.6725 4.7579>, rot=<0.0416 -0.9991 0 0>}
 extra_cup1 {parent=spine_04, loc=<4.7124 12.3043 -6.9455>, rot=<0.0416 -0.9991 0 0>}
 extra_chest_bag_bot {parent=spine_04, loc=<-1.8302 12.7721 0>, rot=<0.0416 -0.9991 0 0>}
 dyn_card_BASE {parent=spine_01, loc=<0.2371 8.0867 -10.4265>, rot=<-0.0115 -0.938 0.3188 -0.1354>}
 dyn_card_01 {parent=dyn_card_BASE, loc=<0 0 0>, rot=<0 0 0 1>}
 extra_pelvis_bag_back {parent=spine_01, loc=<1.0157 -8.4839 7.2111>, rot=<0.1471 -0.9692 0.1674 -0.1051>}
 thigh_l {parent=pelvis, loc=<-2.5693 -0.0109 -7.7501>, rot=<0.0012 0.0572 0.0155 -0.9982>}
 calf_l {parent=thigh_l, loc=<-39.0484 0 0>, rot=<0.0006 -0.0161 -0.0391 -0.9991>}
 foot_l {parent=calf_l, loc=<-42.561 0.1077 0.4292>, rot=<0.0418 -0.0455 0.0343 -0.9975>}
 ball_l {parent=foot_l, loc=<-6.1599 13.8264 0.6806>, rot=<0.0001 -0.0001 -0.7063 -0.7079>}
 calf_twist_02_l {parent=calf_l, loc=<-18.8064 -0.0967 0.0397>, rot=<0.0237 0.0095 -0.0035 -0.9997>}
 calf_twist_01_l {parent=calf_l, loc=<-35.813 -0.0967 0.0397>, rot=<0.0237 0.0095 -0.0035 -0.9997>}
 deform_knee_in_base_l {parent=calf_l, loc=<1.298 -5.2957 -0.0496>, rot=<0 0 0 -1>}
 deform_knee_in_l {parent=deform_knee_in_base_l, loc=<0 0 0>, rot=<0 0 0 1>}
 deform_knee_out_base_l {parent=calf_l, loc=<0 4 0>, rot=<0 0 0 -1>}
 deform_knee_out_l {parent=deform_knee_out_base_l, loc=<0 0 0>, rot=<0 0 0 1>}
 thigh_twist_01_l {parent=thigh_l, loc=<-21.6303 0 0>, rot=<0.0013 0 0 -1>}
 deform_groin_base_l {parent=thigh_l, loc=<-2.1908 8.1142 -1.7442>, rot=<0 0 0 -1>}
 deform_groin_l {parent=deform_groin_base_l, loc=<0 0 0>, rot=<0 0 0 1>}
 thigh_r {parent=pelvis, loc=<-2.5693 -0.0109 7.7501>, rot=<0.0572 -0.0012 0.9982 0.0155>}
 calf_r {parent=thigh_r, loc=<39.0484 0 0>, rot=<0.0006 -0.0161 -0.0391 -0.9991>}
 foot_r {parent=calf_r, loc=<42.561 -0.1077 -0.4292>, rot=<0.0418 -0.0455 0.0343 -0.9975>}
 ball_r {parent=foot_r, loc=<6.1599 -13.8264 -0.6806>, rot=<0.0001 -0.0001 -0.7063 -0.7079>}
 calf_twist_02_r {parent=calf_r, loc=<18.8064 0.0967 -0.0397>, rot=<0.0237 0.0095 -0.0035 -0.9997>}
 calf_twist_01_r {parent=calf_r, loc=<35.813 0.0967 -0.0397>, rot=<0.0237 0.0095 -0.0035 -0.9997>}
 deform_knee_in_base_r {parent=calf_r, loc=<-1.298 5.2957 0.0496>, rot=<0 0 0 -1>}
 deform_knee_in_r {parent=deform_knee_in_base_r, loc=<0 0 0>, rot=<0 0 0 1>}
 deform_knee_out_base_r {parent=calf_r, loc=<0 -4 0>, rot=<0 0 0 -1>}
 deform_knee_out_r {parent=deform_knee_out_base_r, loc=<0 0 0>, rot=<0 0 0 1>}
 thigh_twist_01_r {parent=thigh_r, loc=<21.6303 0 0>, rot=<0.0013 0 0 -1>}
 deform_groin_base_r {parent=thigh_r, loc=<2.1908 -8.1142 1.7442>, rot=<0 0 0 -1>}
 deform_groin_r {parent=deform_groin_base_r, loc=<0 0 0>, rot=<0 0 0 1>}
 deform_glute_base_l {parent=pelvis, loc=<-8.8064 -8.8222 -8.4287>, rot=<0.0275 0.0501 -0.4472 -0.8926>}
 deform_glute_l {parent=deform_glute_base_l, loc=<0 0 0>, rot=<0 0 0 1>}
 deform_glute_base_r {parent=pelvis, loc=<-8.8062 -8.8222 8.4287>, rot=<0.0501 -0.0275 0.8926 -0.4472>}
 deform_glute_r {parent=deform_glute_base_r, loc=<0 0 0>, rot=<0 0 0 1>}
 dyn_katana_BASE {parent=pelvis, loc=<1.0133 -1.9292 16.457>, rot=<0.4149 -0.8988 -0.1397 0.0229>}
 dyn_katana_01 {parent=dyn_katana_BASE, loc=<0 0 0>, rot=<0 0 0 1>}
 dyn_jacket_r_BASE {parent=pelvis, loc=<0.1767 -0.0511 13.8046>, rot=<0.7049 -0.0715 -0.0594 -0.7032>}
 dyn_jacket_r_01 {parent=dyn_jacket_r_BASE, loc=<0 0 0>, rot=<0 0 0 1>}
 dyn_jacket_r_02 {parent=dyn_jacket_r_01, loc=<-4.5784 0 0>, rot=<0 0 0 1>}
 dyn_jacket_b_r_BASE1 {parent=pelvis, loc=<1.6176 -10.2591 5.6653>, rot=<0.1205 -0.0383 -0.0564 -0.9904>}
 dyn_jacket_b_r_01 {parent=dyn_jacket_b_r_BASE1, loc=<0 0 0>, rot=<0 0 0 -1>}
 dyn_jacket_b_r_02 {parent=dyn_jacket_b_r_01, loc=<-4.7757 0 0>, rot=<0 0 0 -1>}
 dyn_jacket_f_r_BASE {parent=pelvis, loc=<-0.0532 8.3187 7.6723>, rot=<0.9951 -0.0219 -0.0289 -0.0916>}
 dyn_jacket_f_r_01 {parent=dyn_jacket_f_r_BASE, loc=<0 0 0>, rot=<0 0 0 -1>}
 dyn_jacket_f_r_02 {parent=dyn_jacket_f_r_01, loc=<-4.9889 0 0>, rot=<0 0 0 -1>}
 dyn_jacket_b_l_BASE {parent=pelvis, loc=<2.2577 -10.1505 -5.365>, rot=<-0.0387 -0.1203 0.9902 -0.0594>}
 dyn_jacket_b_l_01 {parent=dyn_jacket_b_l_BASE, loc=<0 0 0>, rot=<0 0 0 1>}
 dyn_jacket_b_l_02 {parent=dyn_jacket_b_l_01, loc=<4.8318 0 0>, rot=<0 0 0 1>}
 dyn_jacket_l_BASE {parent=pelvis, loc=<3.0292 -0.9024 -13.0737>, rot=<-0.0854 -0.6947 0.7127 -0.0466>}
 dyn_jacket_l_01 {parent=dyn_jacket_l_BASE, loc=<0 0 0>, rot=<0 0 0 1>}
 dyn_jacket_l_02 {parent=dyn_jacket_l_01, loc=<5.3623 0 0>, rot=<0 0 0 1>}
 dyn_jacket_f_l_BASE {parent=pelvis, loc=<2.766 8.4082 -7.5811>, rot=<-0.0219 -0.9951 0.0916 -0.0289>}
 dyn_jacket_f_l_01 {parent=dyn_jacket_f_l_BASE, loc=<0 0 0>, rot=<0 0 0 1>}
 dyn_jacket_f_l_02 {parent=dyn_jacket_f_l_01, loc=<5.5255 0 0>, rot=<0 0 0 1>}
]
