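# Physics asset for the male medium MutedSheath player body cosmetic, with dynamic bodies for jacket, card and katana.
# Six two-segment jacket chains, plus single-segment card and katana, hang from BASE bodies joined to the pelvis with free angular constraints.
# Angular limits keep jacket and card links tight, while the katana swings wider with little twist.
# All shapes are capsules, including pelvis, both thighs and only the right calf.
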PhysicsAsset /BRCosmetics/Characters/Player/Male/Medium/Bodies/M_MED_MutedSheath/Meshes/M_MED_MutedSheath_Physics
bodies {
 calf_r [{shapeType=Capsule, length=36.9127, radius=8.3315, shapeName=calf_r_capsule, center=<29.0385 0 0>, rotation=<rot 90 0 0>}]
 dyn_card_01 [
  {shapeType=Capsule, length=8, radius=0.5, shapeName=dyn_card_01_capsule, center=<3.896 0.1941 0.1441>, rotation=<rot 0 -89.9891 -61.164>}
  {
   shapeType=Capsule
   length=8
   radius=0.5
   shapeName=dyn_card_01_capsule
   center=<3.896 0.1101 0.1441>
   rotation=<rot 60.9024 -179.9883 -179.9949>
  }
 ]
 dyn_card_BASE [{shapeType=Capsule, length=0.505, radius=0.505, shapeName=dyn_card_BASE_capsule, rotation=<rot 0 -90 0>}]
 dyn_jacket_b_l_01 [{shapeType=Capsule, length=9, radius=1, shapeName=dyn_jacket_b_l_01_capsule, center=<4 -1.6723 -3.6884>, rotation=<rot 0 -0.0002 22.0774>}]
 dyn_jacket_b_l_02 [{shapeType=Capsule, length=9, radius=1, shapeName=dyn_jacket_b_l_02_capsule, center=<4 -1.6723 -3.6884>, rotation=<rot 0 -0.0002 22.0774>}]
 dyn_jacket_b_l_BASE [{shapeType=Capsule, length=0.505, radius=0.505, shapeName=dyn_jacket_b_l_BASE_capsule, rotation=<rot 0 -90 0>}]
 dyn_jacket_b_r_01 [{
   shapeType=Capsule
   length=9
   radius=1
   shapeName=dyn_jacket_b_r_01_capsule
   center=<-3.9992 1.1947 1.5723>
   rotation=<rot -0.0222 -0.0917 13.5071>
  }]
 dyn_jacket_b_r_02 [{
   shapeType=Capsule
   length=9
   radius=1
   shapeName=dyn_jacket_b_r_02_capsule
   center=<-3.9992 1.1947 1.5723>
   rotation=<rot -0.0222 -0.0917 13.5071>
  }]
 dyn_jacket_b_r_BASE1 [{shapeType=Capsule, length=0.505, radius=0.505, shapeName=dyn_jacket_b_r_BASE1_capsule, rotation=<rot 0 -90 0>}]
 dyn_jacket_f_l_01 [{shapeType=Capsule, length=9, radius=0.8, shapeName=dyn_jacket_f_l_01_capsule, center=<6 -1.4309 2.2844>, rotation=<rot 0 0.0001 -20.271>}]
 dyn_jacket_f_l_02 [
  {shapeType=Capsule, length=9, radius=1, shapeName=dyn_jacket_f_l_02_capsule, center=<5.736 -1.7284 2.1352>, rotation=<rot 0 0.0001 -20.7988>}
  {shapeType=Capsule, length=9, radius=1, shapeName=dyn_jacket_f_l_02_capsule, center=<5.736 -3.5836 6.1883>, rotation=<rot 0 0 -45.9413>}
 ]
 dyn_jacket_f_l_BASE [{
   shapeType=Capsule
   length=6.5972
   radius=0.9876
   shapeName=dyn_jacket_f_l_BASE_capsule
   center=<0 -0.1716 0.4987>
   rotation=<rot -18.994 -90 0>
  }]
 dyn_jacket_f_r_01 [{shapeType=Capsule, length=9, radius=1.2, shapeName=dyn_jacket_f_r_01_capsule, center=<-4.2 1.0947 -1.3381>, rotation=<rot 0 0 -11.0545>}]
 dyn_jacket_f_r_02 [
  {shapeType=Capsule, length=9, radius=1, shapeName=dyn_jacket_f_r_02_capsule, center=<-4.2 1.1086 -1.6639>, rotation=<rot 0 0.0002 -9.8635>}
  {shapeType=Capsule, length=9, radius=1, shapeName=dyn_jacket_f_r_02_capsule, center=<-4.2 3.925 -6.7065>, rotation=<rot 0 0 -54.0351>}
 ]
 dyn_jacket_f_r_BASE [{shapeType=Capsule, length=0.505, radius=0.505, shapeName=dyn_jacket_f_r_BASE_capsule, rotation=<rot 0 -90 0>}]
 dyn_jacket_l_01 [{shapeType=Capsule, length=9, radius=1.2, shapeName=dyn_jacket_l_01_capsule, center=<6 0 0>}]
 dyn_jacket_l_02 [{shapeType=Capsule, length=9, radius=1.2, shapeName=dyn_jacket_l_02_capsule, center=<6 0 0>}]
 dyn_jacket_l_BASE [{shapeType=Capsule, length=0.505, radius=0.505, shapeName=dyn_jacket_l_BASE_capsule, rotation=<rot 0 -90 0>}]
 dyn_jacket_r_01 [{shapeType=Capsule, length=9, radius=1, shapeName=dyn_jacket_r_01_capsule, center=<-4.2 0 0>}]
 dyn_jacket_r_02 [{shapeType=Capsule, length=9, radius=1.2, shapeName=dyn_jacket_r_02_capsule, center=<-4.2 0 0>}]
 dyn_jacket_r_BASE [{shapeType=Capsule, length=0.505, radius=0.505, shapeName=dyn_jacket_r_BASE_capsule, rotation=<rot 0 -90 0>}]
 dyn_katana_01 [{
   shapeType=Capsule
   length=89.561
   radius=3.1734
   shapeName=dyn_katana_01_capsule
   center=<11.3121 0.5584 0.915>
   rotation=<rot 0 88.4057 -86.221>
  }]
 dyn_katana_BASE [{shapeType=Capsule, length=0.505, radius=0.505, shapeName=dyn_katana_BASE_capsule, rotation=<rot 0 -90 0>}]
 pelvis [{shapeType=Capsule, length=11.4439, radius=7.3108, shapeName=pelvis_capsule, center=<-3.9926 -0.1885 0>}]
 thigh_l [{shapeType=Capsule, length=23, radius=7, center=<-15 0 0>, rotation=<rot 90 0 0>}]
 thigh_r [{shapeType=Capsule, length=23, radius=7, center=<15 0 0>, rotation=<rot 90 0 0>}]
}
constraints [
 {bone1Name=dyn_jacket_r_02, bone2Name=dyn_jacket_r_01, swing1LimitDegrees=10, swing2LimitDegrees=5, twistLimitDegrees=5}
 {bone1Name=dyn_jacket_l_02, bone2Name=dyn_jacket_l_01, swing1LimitDegrees=10, swing2LimitDegrees=5, twistLimitDegrees=5}
 {bone1Name=dyn_jacket_f_r_02, bone2Name=dyn_jacket_f_r_01, swing1LimitDegrees=10, swing2LimitDegrees=5, twistLimitDegrees=5}
 {bone1Name=dyn_jacket_f_l_02, bone2Name=dyn_jacket_f_l_01, swing1LimitDegrees=10, swing2LimitDegrees=2, twistLimitDegrees=5}
 {bone1Name=dyn_jacket_b_r_02, bone2Name=dyn_jacket_b_r_01, swing1LimitDegrees=10, swing2LimitDegrees=5, twistLimitDegrees=5}
 {bone1Name=dyn_jacket_b_l_02, bone2Name=dyn_jacket_b_l_01, swing1LimitDegrees=10, swing2LimitDegrees=5, twistLimitDegrees=5}
 {
  bone1Name=thigh_l
  bone2Name=pelvis
  swing1LimitDegrees=45
  swing1Motion=Free
  swing2LimitDegrees=45
  swing2Motion=Free
  twistLimitDegrees=45
  twistMotion=Free
 }
 {
  bone1Name=thigh_r
  bone2Name=pelvis
  swing1LimitDegrees=45
  swing1Motion=Free
  swing2LimitDegrees=45
  swing2Motion=Free
  twistLimitDegrees=45
  twistMotion=Free
 }
 {bone1Name=dyn_jacket_r_01, bone2Name=dyn_jacket_r_BASE, swing1LimitDegrees=10, swing2LimitDegrees=5, twistLimitDegrees=5}
 {bone1Name=dyn_jacket_b_r_01, bone2Name=dyn_jacket_b_r_BASE1, swing1LimitDegrees=10, swing2LimitDegrees=5, twistLimitDegrees=5}
 {bone1Name=dyn_jacket_f_r_01, bone2Name=dyn_jacket_f_r_BASE, swing1LimitDegrees=10, swing2LimitDegrees=5, twistLimitDegrees=5}
 {bone1Name=dyn_jacket_b_l_01, bone2Name=dyn_jacket_b_l_BASE, swing1LimitDegrees=10, swing2LimitDegrees=5, twistLimitDegrees=5}
 {bone1Name=dyn_jacket_l_01, bone2Name=dyn_jacket_l_BASE, swing1LimitDegrees=10, swing2LimitDegrees=5, twistLimitDegrees=5}
 {bone1Name=dyn_jacket_f_l_01, bone2Name=dyn_jacket_f_l_BASE, swing1LimitDegrees=10, swing2LimitDegrees=2, twistLimitDegrees=5}
 {
  bone1Name=dyn_jacket_b_l_BASE
  bone2Name=pelvis
  swing1LimitDegrees=45
  swing1Motion=Free
  swing2LimitDegrees=45
  swing2Motion=Free
  twistLimitDegrees=45
  twistMotion=Free
 }
 {
  bone1Name=dyn_jacket_b_r_BASE1
  bone2Name=pelvis
  swing1LimitDegrees=45
  swing1Motion=Free
  swing2LimitDegrees=45
  swing2Motion=Free
  twistLimitDegrees=45
  twistMotion=Free
 }
 {
  bone1Name=dyn_jacket_f_l_BASE
  bone2Name=pelvis
  swing1LimitDegrees=45
  swing1Motion=Free
  swing2LimitDegrees=45
  swing2Motion=Free
  twistLimitDegrees=45
  twistMotion=Free
 }
 {
  bone1Name=dyn_jacket_f_r_BASE
  bone2Name=pelvis
  swing1LimitDegrees=45
  swing1Motion=Free
  swing2LimitDegrees=45
  swing2Motion=Free
  twistLimitDegrees=45
  twistMotion=Free
 }
 {
  bone1Name=dyn_jacket_l_BASE
  bone2Name=pelvis
  swing1LimitDegrees=45
  swing1Motion=Free
  swing2LimitDegrees=45
  swing2Motion=Free
  twistLimitDegrees=45
  twistMotion=Free
 }
 {
  bone1Name=dyn_jacket_r_BASE
  bone2Name=pelvis
  swing1LimitDegrees=45
  swing1Motion=Free
  swing2LimitDegrees=45
  swing2Motion=Free
  twistLimitDegrees=45
  twistMotion=Free
 }
 {
  bone1Name=dyn_card_BASE
  bone2Name=pelvis
  swing1LimitDegrees=45
  swing1Motion=Free
  swing2LimitDegrees=45
  swing2Motion=Free
  twistLimitDegrees=45
  twistMotion=Free
 }
 {bone1Name=dyn_card_01, bone2Name=dyn_card_BASE, swing1LimitDegrees=10, swing2LimitDegrees=8, twistLimitDegrees=15}
 {
  bone1Name=dyn_katana_BASE
  bone2Name=pelvis
  swing1LimitDegrees=45
  swing1Motion=Free
  swing2LimitDegrees=45
  swing2Motion=Free
  twistLimitDegrees=45
  twistMotion=Free
 }
 {bone1Name=dyn_katana_01, bone2Name=dyn_katana_BASE, swing1LimitDegrees=25, swing2LimitDegrees=25, twistLimitDegrees=3}
 {
  bone1Name=calf_r
  bone2Name=thigh_r
  swing1LimitDegrees=45
  swing1Motion=Free
  swing2LimitDegrees=45
  swing2Motion=Free
  twistLimitDegrees=45
  twistMotion=Free
 }
]
