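# Post-copy physics AnimBP for the MutedSheath male character's face accessory mesh, adding secondary motion to front and back hair bones.
# The AnimGraph copies the attached parent's pose, then chains single-bone AnimDynamics nodes with angular limits and linear springs.
# Each update computes a shared gravity override through a deprecated helper.
# In the frontend, outside emotes and skydiving, simulation switches to world space at reduced alpha; otherwise alpha is full.
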AnimBlueprint /BRCosmetics/Characters/Player/Male/Medium/Bodies/M_MED_MutedSheath/Meshes/Parts/M_MED_MutedSheath_FaceAcc_AnimBP
target_skeleton=&/M_MED_MutedSheath_FaceAcc_Skeleton
graphs [
 AnimGraph {nodes [
   n1 = Get GravityOverride()
   n2 = Get AnimDynamics_SimSpace_Frontend() -> (AnimDynamics_SimSpace_Frontend:AnimPhysSimSpaceType Enum)
   n3 = Get Anim_Dynamics_Alpha_Frontend() -> (Anim_Dynamics_Alpha_Frontend:double)
   n4 = Anim Dynamics - Bone: dyn__hair_f_01_BASE!(
    SimulationSpace:AnimPhysSimSpaceType Enum=@n2
    ComponentPose=Local To Component(LocalPose=Copy Pose From Mesh() {bUseAttachedParent=true})
    Alpha=@n3
    GravityOverride:Vector=@n1
    ExternalForce="0,0,0"
   ) -> (Pose) {
    angularBiasOverride=0.9999
    angularDampingOverride=0.9999
    bLinearSpring=true
    bOverrideAngularBias=true
    bOverrideAngularDamping=true
    bOverrideLinearDamping=true
    bUseGravityOverride=true
    linearDampingOverride=0.9999
    linearSpringConstant=100
    simulationSpace=RootRelative
    alphaBoolBlend {blendOption=Linear}
    alphaScaleBiasClamp {scale=0.5}
    boundBone {boneName=dyn__hair_f_01_BASE}
    componentAppliedLinearAccClamp=<1250 1250 1250>
    componentLinearAccScale=<0.4 0.5 0.5>
    componentLinearVelScale=<0.4 0.5 0.5>
    physicsBodyDefinitions [{
      boundBone {boneName=dyn__hair_f_01_BASE}
      boxExtents=<12 8 8>
      constraintSetup {angularTargetAxis=AxisZ, angularLimitsMax=<10 15 20>, angularLimitsMin=<-10 -15 -30>, angularTarget=<1 1 1>}
      localJointOffset=<4 0 0>
     }]
   }
   n5 = Anim Dynamics - Bone: dyn__hair_f_02_BASE!(SimulationSpace:AnimPhysSimSpaceType Enum=@n2, ComponentPose=@n4, Alpha=@n3, GravityOverride:Vector=@n1, ExternalForce="0,0,0") -> (Pose) {
    angularBiasOverride=0.9999
    angularDampingOverride=0.9999
    bLinearSpring=true
    bOverrideAngularBias=true
    bOverrideAngularDamping=true
    bOverrideLinearDamping=true
    bUseGravityOverride=true
    linearDampingOverride=0.9999
    linearSpringConstant=100
    simulationSpace=RootRelative
    alphaBoolBlend {blendOption=Linear}
    alphaScaleBiasClamp {scale=0.5}
    boundBone {boneName=dyn__hair_f_02_BASE}
    componentAppliedLinearAccClamp=<1250 1250 1250>
    componentLinearAccScale=<0.4 0.5 0.5>
    componentLinearVelScale=<0.4 0.5 0.5>
    physicsBodyDefinitions [{
      boundBone {boneName=dyn__hair_f_02_BASE}
      boxExtents=<12 8 8>
      constraintSetup {angularTargetAxis=AxisZ, angularLimitsMax=<10 15 20>, angularLimitsMin=<-10 -15 -30>, angularTarget=<1 1 1>}
      localJointOffset=<4 0 0>
     }]
   }
   n6 = Anim Dynamics - Bone: dyn__hair_f_06_BASE!(SimulationSpace:AnimPhysSimSpaceType Enum=@n2, ComponentPose=@n5, Alpha=@n3, GravityOverride:Vector=@n1, ExternalForce="0,0,0") -> (Pose) {
    angularBiasOverride=0.9999
    angularDampingOverride=0.9999
    bLinearSpring=true
    bOverrideAngularBias=true
    bOverrideAngularDamping=true
    bOverrideLinearDamping=true
    bUseGravityOverride=true
    linearDampingOverride=0.9999
    linearSpringConstant=100
    simulationSpace=RootRelative
    alphaBoolBlend {blendOption=Linear}
    alphaScaleBiasClamp {scale=0.5}
    boundBone {boneName=dyn__hair_f_06_BASE}
    componentAppliedLinearAccClamp=<1250 1250 1250>
    componentLinearAccScale=<0.4 0.5 0.5>
    componentLinearVelScale=<0.4 0.5 0.5>
    physicsBodyDefinitions [{
      boundBone {boneName=dyn__hair_f_06_BASE}
      boxExtents=<12 8 8>
      constraintSetup {angularTargetAxis=AxisZ, angularLimitsMax=<10 15 20>, angularLimitsMin=<-10 -15 -30>, angularTarget=<1 1 1>}
      localJointOffset=<4 0 0>
     }]
   }
   n7 = Anim Dynamics - Bone: dyn__hair_f_03_BASE!(SimulationSpace:AnimPhysSimSpaceType Enum=@n2, ComponentPose=@n6, Alpha=@n3, GravityOverride:Vector=@n1, ExternalForce="0,0,0") -> (Pose) {
    angularBiasOverride=0.9999
    angularDampingOverride=0.9999
    bLinearSpring=true
    bOverrideAngularBias=true
    bOverrideAngularDamping=true
    bOverrideLinearDamping=true
    bUseGravityOverride=true
    linearDampingOverride=0.9999
    linearSpringConstant=100
    simulationSpace=RootRelative
    alphaBoolBlend {blendOption=Linear}
    alphaScaleBiasClamp {scale=0.5}
    boundBone {boneName=dyn__hair_f_03_BASE}
    componentAppliedLinearAccClamp=<1250 1250 1250>
    componentLinearAccScale=<0.4 0.5 0.5>
    componentLinearVelScale=<0.4 0.5 0.5>
    physicsBodyDefinitions [{
      boundBone {boneName=dyn__hair_f_03_BASE}
      boxExtents=<12 8 8>
      constraintSetup {angularTargetAxis=AxisZ, angularLimitsMax=<10 15 20>, angularLimitsMin=<-10 -15 -30>, angularTarget=<1 1 1>}
      localJointOffset=<4 0 0>
     }]
   }
   n8 = Anim Dynamics - Bone: dyn__hair_f_04_BASE!(SimulationSpace:AnimPhysSimSpaceType Enum=@n2, ComponentPose=@n7, Alpha=@n3, GravityOverride:Vector=@n1, ExternalForce="0,0,0") -> (Pose) {
    angularBiasOverride=0.9999
    angularDampingOverride=0.9999
    bLinearSpring=true
    bOverrideAngularBias=true
    bOverrideAngularDamping=true
    bOverrideLinearDamping=true
    bUseGravityOverride=true
    linearDampingOverride=0.9999
    linearSpringConstant=100
    simulationSpace=RootRelative
    alphaBoolBlend {blendOption=Linear}
    alphaScaleBiasClamp {scale=0.5}
    boundBone {boneName=dyn__hair_f_04_BASE}
    componentAppliedLinearAccClamp=<1250 1250 1250>
    componentLinearAccScale=<0.4 0.5 0.5>
    componentLinearVelScale=<0.4 0.5 0.5>
    physicsBodyDefinitions [{
      boundBone {boneName=dyn__hair_f_04_BASE}
      boxExtents=<12 8 8>
      constraintSetup {angularTargetAxis=AxisZ, angularLimitsMax=<10 15 20>, angularLimitsMin=<-10 -15 -30>, angularTarget=<1 1 1>}
      localJointOffset=<4 0 0>
     }]
   }
   n9 = Anim Dynamics - Bone: dyn__hair_f_05_BASE!(SimulationSpace:AnimPhysSimSpaceType Enum=@n2, ComponentPose=@n8, Alpha=@n3, GravityOverride:Vector=@n1, ExternalForce="0,0,0") -> (Pose) {
    angularBiasOverride=0.9999
    angularDampingOverride=0.9999
    bLinearSpring=true
    bOverrideAngularBias=true
    bOverrideAngularDamping=true
    bOverrideLinearDamping=true
    bUseGravityOverride=true
    linearDampingOverride=0.9999
    linearSpringConstant=100
    simulationSpace=RootRelative
    alphaBoolBlend {blendOption=Linear}
    alphaScaleBiasClamp {scale=0.5}
    boundBone {boneName=dyn__hair_f_05_BASE}
    componentAppliedLinearAccClamp=<1250 1250 1250>
    componentLinearAccScale=<0.4 0.5 0.5>
    componentLinearVelScale=<0.4 0.5 0.5>
    physicsBodyDefinitions [{
      boundBone {boneName=dyn__hair_f_05_BASE}
      boxExtents=<12 8 8>
      constraintSetup {angularTargetAxis=AxisZ, angularLimitsMax=<10 15 20>, angularLimitsMin=<-10 -15 -30>, angularTarget=<1 1 1>}
      localJointOffset=<4 0 0>
     }]
   }
   n10 = Anim Dynamics - Bone: dyn__hair_b_01_BASE!(SimulationSpace:AnimPhysSimSpaceType Enum=@n2, ComponentPose=@n9, Alpha=@n3, GravityOverride:Vector=@n1, ExternalForce="0,0,0") -> (Pose) {
    angularBiasOverride=0.9999
    angularDampingOverride=0.9999
    bLinearSpring=true
    bOverrideAngularBias=true
    bOverrideAngularDamping=true
    bOverrideLinearDamping=true
    bUseGravityOverride=true
    linearDampingOverride=0.9999
    linearSpringConstant=100
    simulationSpace=RootRelative
    alphaBoolBlend {blendOption=Linear}
    alphaScaleBiasClamp {scale=0.5}
    boundBone {boneName=dyn__hair_b_01_BASE}
    componentAppliedLinearAccClamp=<1250 1250 1250>
    componentLinearAccScale=<0.4 0.5 0.5>
    componentLinearVelScale=<0.4 0.5 0.5>
    physicsBodyDefinitions [{
      boundBone {boneName=dyn__hair_b_01_BASE}
      boxExtents=<12 8 8>
      constraintSetup {angularTargetAxis=AxisZ, angularLimitsMax=<10 15 10>, angularLimitsMin=<-10 -15 -40>, angularTarget=<1 1 1>}
      localJointOffset=<2 3 0>
     }]
   }
   n11 = Anim Dynamics - Bone: dyn__hair_b_02_BASE!(SimulationSpace:AnimPhysSimSpaceType Enum=@n2, ComponentPose=@n10, Alpha=@n3, GravityOverride:Vector=@n1, ExternalForce="0,0,0") -> (Pose) {
    angularBiasOverride=0.9999
    angularDampingOverride=0.9999
    bLinearSpring=true
    bOverrideAngularBias=true
    bOverrideAngularDamping=true
    bOverrideLinearDamping=true
    bUseGravityOverride=true
    linearDampingOverride=0.9999
    linearSpringConstant=100
    simulationSpace=RootRelative
    alphaBoolBlend {blendOption=Linear}
    alphaScaleBiasClamp {scale=0.5}
    boundBone {boneName=dyn__hair_b_02_BASE}
    componentAppliedLinearAccClamp=<1250 1250 1250>
    componentLinearAccScale=<0.4 0.5 0.5>
    componentLinearVelScale=<0.4 0.5 0.5>
    physicsBodyDefinitions [{
      boundBone {boneName=dyn__hair_b_02_BASE}
      boxExtents=<12 8 8>
      constraintSetup {angularTargetAxis=AxisZ, angularLimitsMax=<10 15 10>, angularLimitsMin=<-10 -15 -40>, angularTarget=<1 1 1>}
      localJointOffset=<4 1 0>
     }]
   }
   n12 = Anim Dynamics - Bone: dyn__hair_b_03_BASE!(SimulationSpace:AnimPhysSimSpaceType Enum=@n2, ComponentPose=@n11, Alpha=@n3, GravityOverride:Vector=@n1, ExternalForce="0,0,0") -> (Pose) {
    angularBiasOverride=0.9999
    angularDampingOverride=0.9999
    bLinearSpring=true
    bOverrideAngularBias=true
    bOverrideAngularDamping=true
    bOverrideLinearDamping=true
    bUseGravityOverride=true
    linearDampingOverride=0.9999
    linearSpringConstant=100
    simulationSpace=RootRelative
    alphaBoolBlend {blendOption=Linear}
    alphaScaleBiasClamp {scale=0.5}
    boundBone {boneName=dyn__hair_b_03_BASE}
    componentAppliedLinearAccClamp=<1250 1250 1250>
    componentLinearAccScale=<0.4 0.5 0.5>
    componentLinearVelScale=<0.4 0.5 0.5>
    physicsBodyDefinitions [{
      boundBone {boneName=dyn__hair_b_03_BASE}
      boxExtents=<12 8 8>
      constraintSetup {angularTargetAxis=AxisZ, angularLimitsMax=<10 15 10>, angularLimitsMin=<-10 -15 -40>, angularTarget=<1 1 1>}
      localJointOffset=<2 3 0>
     }]
   }
   n13 = Anim Dynamics - Bone: dyn__hair_b_04_BASE!(SimulationSpace:AnimPhysSimSpaceType Enum=@n2, ComponentPose=@n12, Alpha=@n3, GravityOverride:Vector=@n1, ExternalForce="0,0,0") -> (Pose) {
    angularBiasOverride=0.9999
    angularDampingOverride=0.9999
    bLinearSpring=true
    bOverrideAngularBias=true
    bOverrideAngularDamping=true
    bOverrideLinearDamping=true
    bUseGravityOverride=true
    linearDampingOverride=0.9999
    linearSpringConstant=100
    simulationSpace=RootRelative
    alphaBoolBlend {blendOption=Linear}
    alphaScaleBiasClamp {scale=0.5}
    boundBone {boneName=dyn__hair_b_04_BASE}
    componentAppliedLinearAccClamp=<1250 1250 1250>
    componentLinearAccScale=<0.4 0.5 0.5>
    componentLinearVelScale=<0.4 0.5 0.5>
    physicsBodyDefinitions [{
      boundBone {boneName=dyn__hair_b_04_BASE}
      boxExtents=<12 8 8>
      constraintSetup {angularTargetAxis=AxisZ, angularLimitsMax=<10 15 10>, angularLimitsMin=<-10 -15 -40>, angularTarget=<1 1 1>}
      localJointOffset=<4 1 0>
     }]
   }
   n14 = Anim Dynamics - Bone: dyn__hair_b_05_BASE!(SimulationSpace:AnimPhysSimSpaceType Enum=@n2, ComponentPose=@n13, Alpha=@n3, GravityOverride:Vector=@n1, ExternalForce="0,0,0") -> (Pose) {
    angularBiasOverride=0.9999
    angularDampingOverride=0.9999
    bLinearSpring=true
    bOverrideAngularBias=true
    bOverrideAngularDamping=true
    bOverrideLinearDamping=true
    bUseGravityOverride=true
    linearDampingOverride=0.9999
    linearSpringConstant=100
    simulationSpace=RootRelative
    alphaBoolBlend {blendOption=Linear}
    alphaScaleBiasClamp {scale=0.5}
    boundBone {boneName=dyn__hair_b_05_BASE}
    componentAppliedLinearAccClamp=<1250 1250 1250>
    componentLinearAccScale=<0.4 0.5 0.5>
    componentLinearVelScale=<0.4 0.5 0.5>
    physicsBodyDefinitions [{
      boundBone {boneName=dyn__hair_b_05_BASE}
      boxExtents=<12 8 8>
      constraintSetup {angularTargetAxis=AxisZ, angularLimitsMax=<10 15 10>, angularLimitsMin=<-10 -15 -40>, angularTarget=<1 1 1>}
      localJointOffset=<2 3 0>
     }]
   }
   n15 = Anim Dynamics - Bone: dyn__hair_b_06_BASE!(SimulationSpace:AnimPhysSimSpaceType Enum=@n2, ComponentPose=@n14, Alpha=@n3, GravityOverride:Vector=@n1, ExternalForce="0,0,0") -> (Pose) {
    angularBiasOverride=0.9999
    angularDampingOverride=0.9999
    bLinearSpring=true
    bOverrideAngularBias=true
    bOverrideAngularDamping=true
    bOverrideLinearDamping=true
    bUseGravityOverride=true
    linearDampingOverride=0.9999
    linearSpringConstant=100
    simulationSpace=RootRelative
    alphaBoolBlend {blendOption=Linear}
    alphaScaleBiasClamp {scale=0.5}
    boundBone {boneName=dyn__hair_b_06_BASE}
    componentAppliedLinearAccClamp=<1250 1250 1250>
    componentLinearAccScale=<0.4 0.5 0.5>
    componentLinearVelScale=<0.4 0.5 0.5>
    physicsBodyDefinitions [{
      boundBone {boneName=dyn__hair_b_06_BASE}
      boxExtents=<12 8 8>
      constraintSetup {angularTargetAxis=AxisZ, angularLimitsMax=<10 15 10>, angularLimitsMin=<-10 -15 -40>, angularTarget=<1 1 1>}
      localJointOffset=<4 1 0>
     }]
   }
   Output Pose(Result=Component To Local(ComponentPose=@n15)) {layerGroup=None}
  ]}
 EventGraph {nodes [
   Event BlueprintUpdateAnimation()
   > n1 = Gravity Override State Function DEPRECATED(CharacterPartAnimInstance=Self-Reference(), GameTimeInSeconds=GetGameTimeInSeconds(), GravityOverrideParams=Get Gravity Override Params())
   > Set GravityOverride(GravityOverride=@n1)
   n2 = NOT Boolean(A=Get bIsPlayingEmote())
   n3 = NOT Boolean(A=Get bIsSkyDiving())
   > n4 = Branch(Condition=AND Boolean(A=Get bIsInFrontEnd(), B=AND Boolean(A=@n3, B=@n2)))
   > Set AnimDynamics_SimSpace_Frontend(AnimDynamics_SimSpace_Frontend:AnimPhysSimSpaceType Enum=World) -> (Output_Get:AnimPhysSimSpaceType Enum)
   > Set Anim_Dynamics_Alpha_Frontend(Anim_Dynamics_Alpha_Frontend:double=0.4) -> (Output_Get:double)
   Set AnimDynamics_SimSpace_Frontend(execute=@n4.else)
   > Set Anim_Dynamics_Alpha_Frontend(Anim_Dynamics_Alpha_Frontend=1)
  ]}
]
variables [
 Anim_Dynamics_Alpha_Frontend {type=number}
 AnimDynamics_SimSpace_Frontend {type=AnimPhysSimSpaceType}
 Gravity Override Params {type=GravityOverrideParamsStruct}
 GravityOverride {type=Vector}
]
